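# Material function that layers a glowing, active Fortanium crystal/rock look over an input material, blended through a channel-selected mask.
# Base color mixes darkened source with view-angle curve-atlas tints; roughness and specular are remapped.
# Emissive combines a camera-facing virtual-plane core, thickness map, panning world-aligned noise and quality-switched glitter.
# Custom primitive data supplies power-up stage, timed pulses and impact center, with an Is_Mace variant; many debug and unconnected nodes remain.
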MaterialFunction /FortaniumGeneral/Materials/MF_ActiveFortanium
graph {nodes [
  n1 = SphereMask(A=NamedRerouteUsage(), B=Constant3Vector(), Radius=ScalarParameter() {defaultValue=350, parameterName=PulseMaskRadius}, Hardness=Constant())
  n2 = NamedRerouteUsage()
  Add(A=@n1, B=Multiply(A=ComponentMask(None=@n2) {r=true}, B=ScalarParameter() {parameterName=PulseDistortion_Amt}) {constB=0.1})
  n3 = NamedRerouteUsage()
  n4 = Multiply(A=Time(), B=VectorParameter() {parameterName=NoisePanSpeed, defaultValue=<rgba 0 0 -128 1>}.RGB)
  Add(A=@n3, B=@n4)
  CameraVectorWS()
  n5 = Multiply(A=Time(), B=ScalarParameter() {defaultValue=0.5, parameterName=SineSpeed2})
  n6 = Sine(None=@n5)
  Clamp(None=Multiply(A=ConstantBiasScale(None=@n6)) {constB=2})
  n7 = DotProduct(A=SkyAtmosphereLightDirection(), B=PixelNormalWS())
  ConstantBiasScale(None=@n7)
  n8 = Multiply(A=TextureCoordinate(), B=ScalarParameter() {defaultValue=8, parameterName=SparkleTile})
  n9 = Multiply(A=TextureSample(UVs=@n8) {samplerType=SAMPLERTYPE_LinearColor, texture=/Game/Global/Textures/Default/Noise/T_Apollo_TCNoise}.RGB) {constB=2}
  DotProduct(A=CameraVectorWS(), B=Transform(None=Subtract(A=@n9)))
  n10 = Time()
  Frac(None=Multiply(A=@n10, B=ScalarParameter() {defaultValue=0.5, parameterName=DebugFracTime}))
  Frac(None=Multiply(A=@n10, B=ScalarParameter() {defaultValue=2, parameterName=PulseSpeed}) {constB=0.25})
  Frac(None=@n5)
  Frac(None=Multiply(A=Time()) {constB=0.2})
  n11 = GetMaterialAttributes(None=FunctionInput() {inputName=Material, inputType=FunctionInput_MaterialAttributes}) {attributeGetTypes [
    #1
    #2
    #3
    #4
   ]}
  n12 = NamedRerouteUsage()
  n13 = ScalarParameter() {defaultValue=0.5, parameterName=SourceColor_Darken}
  n14 = SetMaterialAttributes!(
   MaterialAttributes=@n11.MaterialAttributes
   BaseColor=LinearInterpolate(
    A=Multiply(A=@n12, B=@n13) {constB=0.5}
    B=NamedRerouteUsage()
    Alpha=ScalarParameter() {defaultValue=0.45, parameterName=FortaniumColorBlend_AMt}
   ) {constAlpha=0.65}
   Metallic=Constant()
   Specular=NamedRerouteUsage()
   Roughness=NamedRerouteUsage()
   EmissiveColor=Add(A=NamedRerouteUsage(), B=NamedRerouteUsage())
  ) {attributeSetTypes [
    #1
    #5
    #2
    #3
    #4
   ]}
  n15 = FunctionInput(Preview=Constant()) {bUsePreviewValueAsDefault=true, inputName=Mask}
  FunctionOutput(None=BlendMaterialAttributes(
    A=@n11.MaterialAttributes
    B=@n14
    Alpha=ChannelMaskParameter(None=@n15) {maskChannel=Blue, parameterName=ActiveFortaniumMaskChannel, defaultValue=<rgba 0 0 1 0>}.R
   ))
  LinearInterpolate(
   A=ScalarParameter() {defaultValue=2, parameterName=BaseColorShadowDarkness}
   B=ScalarParameter() {defaultValue=3, parameterName=BaseColorSunBrightness}
   Alpha=NamedRerouteUsage()
  )
  n16 = VectorParameter() {parameterName=PulseColor, defaultValue=<rgba 0.3631 0 1 1>}
  n17 = ScalarParameter() {bUseCustomPrimitiveData=true, parameterName=FortaniumPulse, primitiveDataIndex=1}
  n18 = CurveAtlasRowParameter(CurveTime=@n17) {
   atlas=&/FortaniumCurveAtlas
   curve=&/FortaniumPulseCurve
   defaultValue=3
   parameterName=PulseCurve
  }
  n19 = Multiply(A=Add(A=@n1, B=@n18.G), B=@n18.R)
  n20 = NamedRerouteDeclaration(None=SphereMask(
    A=Frac(None=@n19)
    B=Constant() {r=0.5}
    Radius=ScalarParameter() {defaultValue=0.5, parameterName=PulseRadius}
    Hardness=ScalarParameter() {parameterName=PulseHardness}
   )) {name=PulseMask, nodeColor=<rgba 1 0 0.4687 1>}
  LinearInterpolate(A=VectorParameter() {parameterName=PulseColor2, defaultValue=<rgba 0 0.6513 1 1>}.RGB, B=@n16.RGB, Alpha=Power(Base=@n20))
  n21 = TextureSample(UVs=Multiply(A=TextureCoordinate(), B=ScalarParameter() {defaultValue=4, group=GlitterAnimated, parameterName=Sparkle Tiling, sliderMax=32})) {samplerSource=SSM_Wrap_WorldGroupSettings, samplerType=SAMPLERTYPE_Masks, texture=/Game/Weapons/WeaponSkins/Textures/T_SparkleMask}
  LinearInterpolate(
   A=@n21.RGB
   Alpha=MaterialFunctionCall!(
    Blend Range=ScalarParameter() {defaultValue=50, group=Glitter, parameterName=Blend Distance, sortPriority=4}
    Start Offset=ScalarParameter() {defaultValue=-150, group=Glitter, parameterName=Blend Offset, sortPriority=5}
   ) {materialFunction=/Engine/Functions/Engine_MaterialFunctions03/Particles/Distance_Blend}.Result
  )
  n22 = DotProduct(A=CameraVectorWS(), B=LinearInterpolate(A=PixelNormalWS(), B=VertexNormalWS(), Alpha=ScalarParameter() {parameterName=FT_BlendVertexNormal}))
  n23 = Add(A=ConstantBiasScale(None=NamedRerouteDeclaration(None=@n22) {name=Fresnel, nodeColor=<rgba 1 0 0.3516 1>}), B=NamedRerouteUsage())
  n24 = Clamp(None=Power(Base=Frac(None=@n23)))
  LinearInterpolate(
   A=VectorParameter() {parameterName=Color1, defaultValue=<rgba 0.3913 0.128 0.6146 1>}.RGB
   B=VectorParameter() {parameterName=Color2, defaultValue=<rgba 0 0.2358 0.6146 1>}.RGB
   Alpha=@n24
  )
  n25 = LinearInterpolate(
   A=VectorParameter() {parameterName=EmisiveColor1, defaultValue=<rgba 0 0.316 1 1>}.RGB
   B=VectorParameter() {parameterName=EmisiveColor2, defaultValue=<rgba 0.6919 0.0235 1 1>}.RGB
   Alpha=NamedRerouteUsage()
  )
  n26 = Multiply(A=@n25, B=ScalarParameter() {defaultValue=5, parameterName=MAceImpactGLow})
  LinearInterpolate(A=@n25, B=@n26, Alpha=NamedRerouteUsage())
  n27 = ScalarParameter() {parameterName=FT_Specular_MinSpec}
  n28 = NamedRerouteUsage()
  LinearInterpolate(A=@n27, B=NamedRerouteUsage(), Alpha=@n28)
  n29 = ScalarParameter() {defaultValue=0.2, parameterName=FT_Roughness_Min}
  n30 = NamedRerouteUsage()
  LinearInterpolate(A=@n29, B=NamedRerouteUsage(), Alpha=@n30)
  MaterialFunctionCall(Plane Normal Axis=CameraVectorWS(), Plane Center=ObjectPositionWS(), UV Scale=Constant() {r=1}) {materialFunction=&VirtualPlaneCoordinates}
  MaterialFunctionCall(
   Normal Vector=PixelNormalWS()
   Camera Vector=CameraVectorWS()
   Invert Fresnel=StaticBool()
   Power=ScalarParameter() {defaultValue=12, parameterName=EmissiveFresnelPower}
   Cheap contrast dark=ScalarParameter() {parameterName=Contrast_Dark}
   Cheap contrast bright=ScalarParameter() {parameterName=ContrastBright}
  ) {materialFunction=&Fresnel_Function}
  n31 = TextureObjectParameter() {
   group=Detail Texture
   parameterName=Detail N
   samplerType=SAMPLERTYPE_Normal
   sortPriority=2
   texture=/Game/Environments/Apollo/Rocks/Natural_Rocks/Textures/T_RockDetails_02_N
  }
  n32 = ScalarParameter() {defaultValue=768, group=Detail Texture, parameterName=DetailTile}
  n33 = NamedRerouteUsage()
  MaterialFunctionCall(TextureObject=@n31, TextureSize=@n32, Normal=@n33) {materialFunction=&WorldAlignedNormal}
  MaterialFunctionCall() {materialFunction=&ObjectPivotPoint}
  n34 = ScalarParameter() {bUseCustomPrimitiveData=true, parameterName=FortaniumRockPulse, primitiveDataIndex=2}
  MaterialFunctionCall(Number=@n34, UVs=NamedRerouteUsage()) {materialFunction=&DebugScalarValues}
  n35 = CurveAtlasRowParameter(CurveTime=@n24) {atlas=&FortaniumCurveAtlas, curve=&/FortaniumColor_01, parameterName=ColorCurve}
  n36 = ScalarParameter() {defaultValue=10, parameterName=GlitterBrightness}
  NamedRerouteDeclaration(None=LinearInterpolate(A=@n35, B=Multiply(A=@n35, B=@n36), Alpha=NamedRerouteUsage())) {name=BaseColor, nodeColor=<rgba 1 0 0.1875 1>}
  NamedRerouteDeclaration(None=Clamp(None=@n7)) {name=LightDirectionMask, nodeColor=<rgba 0.4062 1 0 1>}
  NamedRerouteDeclaration(None=@n11.Roughness) {name=Source_Roughness, nodeColor=<rgba 0.3828 1 0 1>}
  NamedRerouteDeclaration(None=@n11.BaseColor) {name=Source_BaseColor, nodeColor=<rgba 0.1016 1 0 1>}
  n37 = VectorParameter() {parameterName=VP_SocketLocation, defaultValue=<rgba 0 0 0 1>}
  n38 = Add(
   A=MaterialFunctionCall() {materialFunction=&ObjectPivotPoint}.Object Pivot Location
   B=VectorParameter() {parameterName=VP_Pivot_Offset, defaultValue=<rgba 0 24 64 1>}.RGB
  )
  n39 = MaterialFunctionCall(
   Plane Normal Axis=MaterialFunctionCall() {materialFunction=&CameraDirectionVector}.Result
   Plane Center=NamedRerouteDeclaration(None=StaticSwitchParameter(True=@n37.RGB, False=@n38) {parameterName=UseSocketLocation}) {name=ObjectCenterLocation, nodeColor=<rgba 0 1 0.1094 1>}
   UV Scale=Multiply(A=ScalarParameter() {defaultValue=1, parameterName=CoreScale}, B=ObjectRadius())
  ) {materialFunction=&VirtualPlaneCoordinates}
  n40 = SphereMask(
   A=NamedRerouteDeclaration(None=@n39.World Position - Centered) {name=VP_Coordinates, nodeColor=<rgba 0.5937 1 0 1>}
   B=Constant3Vector()
   Radius=LinearInterpolate(
    A=ScalarParameter() {defaultValue=250, parameterName=EmissiveRadius2}
    B=ScalarParameter() {defaultValue=350, parameterName=EmissiveRadius}
    Alpha=NamedRerouteUsage()
   )
   Hardness=ScalarParameter() {defaultValue=0.25, parameterName=EmissiveSphereHArdness}
  ) {attenuationRadius=1}
  n41 = Multiply(A=NamedRerouteUsage(), B=ScalarParameter() {defaultValue=1, parameterName=PanningNoiseStrength})
  NamedRerouteDeclaration(None=Clamp(None=Subtract(A=@n40, B=@n41))) {name=EmissiveMask, nodeColor=<rgba 0.4062 0 1 1>}
  NamedRerouteDeclaration(None=@n11.EmissiveColor) {name=Source_Emissive, nodeColor=<rgba 1 0.4688 0 1>}
  n42 = PreSkinnedPosition()
  NamedRerouteDeclaration(None=ShaderStageSwitch(PixelShader=VertexInterpolator(VS=@n42).PS, VertexShader=@n42)) {name="PreskinnedPos\\", nodeColor=<rgba 0 1 0.4375 1>}
  n43 = Sine(None=Add(A=NamedRerouteUsage()) {constB=-0.25})
  n44 = NamedRerouteUsage()
  n45 = NamedRerouteUsage()
  n46 = StaticSwitchParameter(
   True=LinearInterpolate(A=@n25, B=@n26, Alpha=ConstantBiasScale(None=@n43))
   False=LinearInterpolate(A=@n25, B=NamedRerouteUsage(), Alpha=Multiply(A=@n44, B=@n45))
  ) {parameterName=Is_Mace}
  n47 = TextureSampleParameter2D() {
   parameterName=ThicknessMap
   samplerSource=SSM_Wrap_WorldGroupSettings
   samplerType=SAMPLERTYPE_LinearGrayscale
   texture=&/s_apollo_rockpile_04_Thickness_Map_from_Mesh
  }
  n48 = ScalarParameter() {defaultValue=1, parameterName=ThicknessSample}
  n49 = ScalarParameter() {defaultValue=0.4, parameterName=ThicknessSampleRadius}
  n50 = ScalarParameter() {parameterName=ThicknessSampleHardness}
  n51 = ScalarParameter() {defaultValue=0.25, parameterName=NoiseThickness_Influence}
  n52 = Multiply(A=NamedRerouteUsage(), B=ScalarParameter() {defaultValue=0.75, parameterName=Pow_02})
  n53 = Clamp(None=Multiply(A=SphereMask(A=@n47.R, B=@n48, Radius=@n49, Hardness=@n50), B=LinearInterpolate(A=@n51, Alpha=@n52)))
  n54 = Multiply(A=@n53, B=LinearInterpolate(Alpha=NamedRerouteUsage()) {constA=0.25})
  n55 = NamedRerouteUsage()
  n56 = NamedRerouteUsage()
  n57 = Multiply(A=@n53, B=NamedRerouteUsage())
  n58 = Power(Base=Multiply(A=Multiply(A=@n46, B=@n54), B=LinearInterpolate(A=@n55, B=@n56, Alpha=@n57)))
  n59 = ScalarParameter() {defaultValue=10, parameterName=FR_PulseGlowStrength}
  n60 = NamedRerouteUsage()
  n61 = Ceil(None=NamedRerouteUsage())
  NamedRerouteDeclaration(None=LinearInterpolate(A=@n58, B=Multiply(A=@n58, B=@n59) {constB=5}, Alpha=Multiply(A=@n60, B=@n61))) {name=Emissive, nodeColor=<rgba 1 0.9375 0 1>}
  n62 = MaterialFunctionCall(Float3=MaterialFunctionCall(
    TextureObject=TextureObject() {samplerType=SAMPLERTYPE_Masks, texture=/Game/Global/Textures/Default/Noise/T_Metal_Noise}
    TextureSize=ScalarParameter() {defaultValue=256, group=Detail Texture, parameterName=DetailTile_02}
    WorldPosition=NamedRerouteUsage()
    World Space Normal=@n33
   ) {materialFunction=/Game/Weapons/WeaponSkins/Wraps/MaterialFunctions/WorldAlignedTexture}.XYZ Texture) {materialFunction=&BreakOutFloat3Components}
  n63 = NamedRerouteDeclaration(None=Clamp(None=Power(Base=@n62.R) {constExponent=3})) {name=DetailNoise_G, nodeColor=<rgba 0 1 0.0481 1>}
  n64 = MaterialFunctionCall(
   Plane Normal Axis=MaterialFunctionCall() {materialFunction=&CameraDirectionVector}.Result
   Plane Center=Add(
    A=MaterialFunctionCall() {materialFunction=&ObjectPivotPoint}.Object Pivot Location
    B=VectorParameter() {parameterName=VP_Pivot_Offset, defaultValue=<rgba 0 24 64 1>}.RGB
   )
   UV Scale=Multiply(A=ScalarParameter() {defaultValue=1, parameterName=Scale}, B=ObjectRadius())
  ) {materialFunction=&VirtualPlaneCoordinates}
  n65 = NamedRerouteUsage()
  n66 = MaterialFunctionCall(Float2=MaterialFunctionCall(
    TextureObject=TextureObject() {texture=/Game/Global/Textures/Default/Noise/T_Noise_Cloud}
    TextureSize=ScalarParameter() {defaultValue=256, group=Detail Texture, parameterName=DetailTile_03}
    WorldPosition=Add(A=@n64.World Position - Centered, B=@n4)
    World Space Normal=@n65
   ) {materialFunction=&WorldAlignedTexture}.XYZ Texture) {materialFunction=&BreakOutFloat2Components}
  NamedRerouteDeclaration(None=Clamp(None=Multiply(A=@n63, B=@n66.R), Min=ScalarParameter() {parameterName=MinGlow})) {name=PanningNoise, nodeColor=<rgba 0 0.6016 1 1>}
  n67 = MaterialFunctionCall(Float2=MaterialFunctionCall(
    TextureObject=TextureObject() {texture=/Game/Characters/Player/Male/Medium/Bodies/M_MED_Viper/Textures/Glitch/T_SoftNoise}
    TextureSize=ScalarParameter() {defaultValue=512, group=Detail Texture, parameterName=DetailTile_04}
    WorldPosition=@n3
    World Space Normal=@n65
   ) {materialFunction=&WorldAlignedTexture}.XYZ Texture) {materialFunction=&BreakOutFloat2Components}
  NamedRerouteDeclaration(None=LinearInterpolate(A=ScalarParameter() {defaultValue=0.25, parameterName=MinGlowNoise}, Alpha=@n67.R)) {name=ThicknessMask, nodeColor=<rgba 1 0 0.7969 1>}
  NamedRerouteDeclaration(None=@n67.R) {name=ThicknessMask_02, nodeColor=<rgba 0.5 1 0 1>}
  n68 = NamedRerouteUsage()
  n69 = Divide(A=@n68) {constB=1}
  n70 = MaterialFunctionCall!(CoordinateOffset=Add(A=ComponentMask(None=@n69) {b=true, r=true}) {constB=0}, Coordinate) {materialFunction=/Game/Environments/Artemis/Sets/Daisy/HyperBloom/Materials/MF_PolarCoordinates}
  n71 = ScalarParameter() {defaultValue=3.14, parameterName=Pi}
  n72 = Sine(None=MaterialFunctionCall(Float2=Multiply(A=@n70.Result, B=@n71)) {materialFunction=&BreakOutFloat2Components}.R)
  n73 = MaterialFunctionCall(Coordinate=Add(A=ComponentMask(None=@n69) {b=true, g=true}) {constB=0}) {materialFunction=&MF_PolarCoordinates}
  n74 = Sine(None=MaterialFunctionCall(Float2=Multiply(A=@n73.Result, B=@n71)) {materialFunction=&BreakOutFloat2Components}.R)
  n75 = Transform(None=Constant3Vector() {constant=<rgba 0 0 1 0>})
  n76 = ComponentMask(None=@n75) {r=true}
  n77 = MaterialFunctionCall(Coordinate=Add(A=ComponentMask(None=@n69) {g=true, r=true}) {constB=0}) {materialFunction=&MF_PolarCoordinates}
  n78 = Sine(None=MaterialFunctionCall(Float2=Multiply(A=@n77.Result, B=@n71)) {materialFunction=&BreakOutFloat2Components}.R)
  n79 = LinearInterpolate(
   A=LinearInterpolate(A=Clamp(None=@n72), B=Clamp(None=@n74), Alpha=Abs(None=@n76))
   B=Clamp(None=@n78)
   Alpha=Abs(None=ComponentMask(None=@n75) {b=true})
  )
  n80 = NamedRerouteUsage()
  NamedRerouteDeclaration(None=Add(A=Multiply(A=@n79, B=@n80), B=NamedRerouteUsage())) {name=VP_Spheremask, nodeColor=<rgba 0.7109 1 0 1>}
  n81 = NamedRerouteUsage()
  n82 = ScalarParameter() {defaultValue=2, parameterName=RockDetailTexturePow}
  n83 = Subtract(A=Multiply(A=Power(Base=@n81, Exp=@n82) {constExponent=4}) {constB=2})
  n84 = ScalarParameter() {defaultValue=0.2, parameterName=DetailNoiseInfluence}
  NamedRerouteDeclaration(None=Clamp(None=Multiply(A=@n83, B=@n84)) {minDefault=-1}) {name=Noise_UVOffset, nodeColor=<rgba 0 1 0.2031 1>}
  n85 = Multiply(A=Time(), B=ScalarParameter() {defaultValue=0.25, parameterName=DebugSineSpeed})
  NamedRerouteDeclaration(None=ConstantBiasScale(None=Sine(None=@n85))) {name=DebugSine, nodeColor=<rgba 1 0.6797 0 1>}
  n86 = Constant3Vector()
  n87 = ScalarParameter() {defaultValue=128, parameterName=Radius}
  n88 = ScalarParameter() {defaultValue=0.25, parameterName=JHardnes}
  n89 = NamedRerouteDeclaration(None=Power(Base=SphereMask(A=@n68, B=@n86, Radius=@n87, Hardness=@n88))) {name=SphereMask, nodeColor=<rgba 0 1 0.5547 1>}
  n90 = SphereMask(
   A=@n68
   B=Constant3Vector()
   Radius=ScalarParameter() {defaultValue=512, parameterName=LightBeamRadius}
   Hardness=ScalarParameter() {defaultValue=0.5, parameterName=LightBeamJHardnes}
  )
  NamedRerouteDeclaration(None=Multiply(A=OneMinus(None=@n89), B=@n90)) {name=LightBeamMask, nodeColor=<rgba 1 0.5859 0 1>}
  n91 = ScalarParameter() {defaultValue=1.2, parameterName=VP_Sphere_Strength}
  NamedRerouteDeclaration(None=Power(Base=Multiply(A=@n40, B=@n91), Exp=ScalarParameter() {defaultValue=4, parameterName=VP_Sphere_Pow})) {name=VP_Core, nodeColor=<rgba 0.2656 1 0 1>}
  n92 = PreSkinnedNormal()
  NamedRerouteDeclaration(None=ShaderStageSwitch(PixelShader=VertexInterpolator(VS=@n92).PS, VertexShader=@n92)) {name=PreSkinnedNormal, nodeColor=<rgba 0 1 0.6719 1>}
  n93 = CurveAtlasRowParameter(CurveTime=NamedRerouteUsage()) {
   atlas=&FortaniumCurveAtlas
   curve=&/FortaniumPowerUpCurve
   defaultValue=1
   parameterName=PowerupCurve
  }
  NamedRerouteDeclaration(None=Multiply(A=@n93.R)) {name=EmissiveColorSwitch, nodeColor=<rgba 0.5703 1 0 1>}
  NamedRerouteDeclaration(None=Multiply(A=@n93.G)) {name=EmissivePower, nodeColor=<rgba 0.2891 1 0 1>}
  NamedRerouteDeclaration(None=Multiply(A=@n93.B)) {name=VP_EmissivePower, nodeColor=<rgba 1 0.4453 0 1>}
  NamedRerouteDeclaration(None=Divide(
    A=ScalarParameter() {bUseCustomPrimitiveData=true, parameterName=FortaniumPowerUp}
    B=ScalarParameter() {defaultValue=3, parameterName=No_of_Stages}
   )) {name=Powerup, nodeColor=<rgba 0.875 0 1 1>}
  n94 = TextureObjectParameter() {
   group=Detail Texture
   parameterName=Detail D
   sortPriority=1
   texture=/Game/Environments/Apollo/Rocks/Natural_Rocks/Textures/T_RockDetails_02D
  }
  n95 = MaterialFunctionCall(Float3=MaterialFunctionCall(TextureObject=@n94, TextureSize=@n32, WorldPosition=NamedRerouteUsage(), World Space Normal=@n33) {materialFunction=&WorldAlignedTexture}.XYZ Texture) {materialFunction=&BreakOutFloat3Components}
  NamedRerouteDeclaration(None=@n95.R) {name=DetailNoise_R, nodeColor=<rgba 1 0.0516 0 1>}
  n96 = Multiply(A=ScalarParameter() {defaultValue=15, parameterName=PulseStrength}, B=Multiply(A=@n16.RGB, B=Clamp(None=@n20) {minDefault=0.5}))
  n97 = OneMinus(None=NamedRerouteUsage())
  NamedRerouteDeclaration(None=Multiply(A=@n96, B=Power(Base=@n97))) {name=PulseGlow, nodeColor=<rgba 1 0.4453 0 1>}
  n98 = Subtract(A=Time(), B=ScalarParameter() {bUseCustomPrimitiveData=true, defaultValue=1, parameterName=FortaniumPulse, primitiveDataIndex=1})
  n99 = ScalarParameter() {defaultValue=0.5, parameterName=MacePulseDuration}
  NamedRerouteDeclaration(None=Saturate(None=Divide(A=@n98, B=@n99))) {name=MaceImpactPulse, nodeColor=<rgba 1 0.5156 0 1>}
  n100 = ScalarParameter() {defaultValue=16, group=GlitterAnimated, parameterName=Sparkle Detail Tiling, sliderMax=64, sliderMin=8, sortPriority=2}
  n101 = Multiply(A=TextureCoordinate(), B=@n100) {constB=3}
  n102 = TextureSampleParameter2D(UVs=@n101) {
   group=GlitterAnimated
   parameterName=Sparkle Mask
   samplerSource=SSM_Wrap_WorldGroupSettings
   samplerType=SAMPLERTYPE_Masks
   sortPriority=0
   texture=&T_SparkleMask
  }
  n103 = Frac(None=@n101)
  n104 = MaterialFunctionCall(X=@n102.R, Y=@n102.R, Z=Divide(A=Subtract(A=@n101, B=@n103), B=@n100)) {desc=could be removed if no blending is needed, materialFunction=&MakeFloat3}
  n105 = ComponentMask(None=@n104.Result) {g=true}
  n106 = MaterialFunctionCall() {materialFunction=&CameraDirectionVector}
  n107 = PixelNormalWS()
  n108 = Multiply(
   A=Add(A=ComponentMask(None=@n104.Result) {b=true}, B=DotProduct(A=@n106.Result, B=@n107))
   B=ScalarParameter() {defaultValue=6, group=GlitterAnimated, parameterName=Glitter Scale, sliderMax=2, sortPriority=3}
  ) {constB=0.5}
  n109 = MaterialFunctionCall(UV Channel=Frac(None=Add(A=@n105, B=@n108))) {materialFunction=&LinearGradient}
  n110 = Add(A=Multiply(A=Sine(None=@n109.VGradient)) {constB=0.5}) {constB=0.5}
  n111 = ScalarParameter() {defaultValue=32, group=GlitterAnimated, parameterName=Fade Curve Exponent, sortPriority=4}
  n112 = Power(Base=Multiply(A=ComponentMask(None=@n104.Result) {r=true}, B=Power(Base=@n110, Exp=@n111) {constExponent=32})) {constExponent=4}
  n113 = Multiply(
   A=Saturate(None=Multiply(A=@n112) {constB=4})
   B=ScalarParameter() {defaultValue=10, group=GlitterAnimated, parameterName=Emissive Intensity, sliderMax=20, sortPriority=5}
  )
  n114 = Constant()
  NamedRerouteDeclaration(None=ShadingPathSwitch(Default=QualitySwitch(Default=@n113, Low=@n114, Medium=@n114), Mobile=QualitySwitch(Default=@n114))) {name=GlitterMask, nodeColor=<rgba 0 1 0.8125 1>}
  NamedRerouteDeclaration(None=@n17) {name=PulseEnable, nodeColor=<rgba 1 0 0.2578 1>}
  n115 = SphereMask(
   A=WorldPosition().XYZ
   B=VectorParameter() {bUseCustomPrimitiveData=true, group=z_code, parameterName=WSDeformationCenter, primitiveDataIndex=20}.RGB
   Radius=ObjectRadius()
   Hardness=Constant()
  )
  n116 = NamedRerouteUsage()
  n117 = ScalarParameter() {defaultValue=0.05, parameterName=ImpactPulseRingRadius}
  n118 = ScalarParameter() {parameterName=ImpactPulseRingHardness}
  NamedRerouteDeclaration(None=Multiply(A=@n115, B=SphereMask(A=@n115, B=@n116, Radius=@n117, Hardness=@n118))) {name=FR_ImpackPulseMask, nodeColor=<rgba 0.1484 0 1 1>}
  n119 = Divide(A=Subtract(A=Time(), B=@n34), B=ScalarParameter() {defaultValue=0.5, parameterName=MacePulseDuration})
  NamedRerouteDeclaration(None=OneMinus(None=Saturate(None=@n119))) {name=RockImpactPulse, nodeColor=<rgba 1 0 0.7266 1>}
  NamedRerouteDeclaration(None=@n25) {name=EMissiveColor, nodeColor=<rgba 0.2891 0 1 1>}
  NamedRerouteDeclaration(None=@n39.UVs) {name=VP_UVs, nodeColor=<rgba 0.8984 0 1 1>}
  n120 = Multiply(A=@n95.B) {constB=1.5}
  NamedRerouteDeclaration(None=Clamp(None=Power(Base=@n120) {constExponent=12})) {name=DetailNoise_B, nodeColor=<rgba 0 0.1289 1 1>}
  NamedRerouteDeclaration(None=LinearInterpolate(A=@n27, B=ScalarParameter() {defaultValue=1, parameterName=FT_Specular_Max}, Alpha=@n28)) {name=Specular, nodeColor=<rgba 1 0.0703 0 1>}
  NamedRerouteDeclaration(None=LinearInterpolate(A=@n29, B=ScalarParameter() {defaultValue=0.4, parameterName=FT_Roughness_Max}, Alpha=@n30)) {name=Roughness, nodeColor=<rgba 1 0 0.6562 1>}
  NamedRerouteDeclaration(None=@n11.Specular) {name=Source_Specular, nodeColor=<rgba 0 1 0.1094 1>}
  NamedRerouteUsage()
  NamedRerouteUsage()
  NamedRerouteUsage()
  NamedRerouteUsage()
  NamedRerouteUsage()
  NamedRerouteUsage()
  NamedRerouteUsage()
  NamedRerouteUsage()
  NamedRerouteUsage()
  ObjectPositionWS()
  ObjectPositionWS()
  Power(Base=ConstantBiasScale(None=@n78))
  Power(Base=ConstantBiasScale(None=@n74))
  Power(Base=ConstantBiasScale(None=@n72))
  Power(Base=@n96)
  Power(Base=NamedRerouteUsage()) {constExponent=1}
  Power(Base=Multiply(A=@n95.G) {constB=1.5}, Exp=ScalarParameter() {defaultValue=12, parameterName=DetailNoise_G_Pow}) {constExponent=4}
  Power(Base=@n90)
  ScalarParameter() {defaultValue=3, parameterName=DebugFortaniumPowerUp}
  ScalarParameter() {defaultValue=256, parameterName=FR_ImpactRadius}
  ScalarParameter() {defaultValue=-1.0000000138484279e+24, parameterName=FR_GameTimeFromBP}
  ScalarParameter() {defaultValue=0.1, parameterName=PulseDebug}
  n121 = NamedRerouteUsage()
  SetMaterialAttributes!(MaterialAttributes, BaseColor=@n121, EmissiveColor=@n121) {attributeSetTypes [#1, #4]}
  Subtract(A=WorldPosition().XYZ, B=NamedRerouteUsage())
  TextureObject() {texture=&T_NoisePack_01}
  TextureObject() {texture=&T_Noise_Cloud}
  TextureObject() {texture=/Game/Effects/Fort_Effects/Textures/T_Cracks_Gradient}
  TextureSample(UVs=@n8) {
   samplerSource=SSM_Wrap_WorldGroupSettings
   samplerType=SAMPLERTYPE_Normal
   texture=/Game/Environments/Asteria/Terrain/Textures/T_SnowSparkleCircles_N
  }
  Time()
  VectorParameter() {parameterName=FR_HitLocation, defaultValue=<rgba 0 0 0 1>}
 ]}
